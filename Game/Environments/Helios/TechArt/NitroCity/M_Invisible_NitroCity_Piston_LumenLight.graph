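# Unlit material for a NitroCity piston light, named as an invisible Lumen light source, that computes a pulsing colored glow.
# A time-driven sine, phase-offset by the actor's world Y position, is biased, made absolute and remapped to modulate light color and brightness.
# A ray tracing quality switch applies Lumen Light Boost only on the ray-traced path, then Light Flicker scales the result.
# The graph shows no link to a material output.
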Material /Game/Environments/Helios/TechArt/NitroCity/M_Invisible_NitroCity_Piston_LumenLight
graph {
 properties {pixelDepthOffsetMode=PDOM_Legacy, shadingModel=MSM_Unlit, usedShadingModels=Unlit}
 nodes [
  n1 = MaterialFunctionCall(Float3=ActorPositionWS()) {materialFunction=&BreakOutFloat3Components}
  n2 = Time()
  n3 = ConstantBiasScale(None=Sine(None=Add(A=@n1.G, B=@n2)) {period=20}) {bias=0.5}
  n4 = Multiply(
   A=Multiply(
    A=VectorParameter() {parameterName=Light Color, defaultValue=<rgba 1 0.1 0 1>}.RGB
    B=ScalarParameter() {defaultValue=1, parameterName=Light Brightness}
   )
   B=LinearInterpolate(Alpha=Abs(None=@n3)) {constA=0.66, constB=1.25}
  )
  n5 = ScalarParameter() {defaultValue=5, parameterName=Lumen Light Boost}
  Multiply(A=ScalarParameter() {defaultValue=1, parameterName=Light Flicker}, B=RayTracingQualitySwitch(Normal=@n4, RayTraced=Multiply(A=@n4, B=@n5)))
  WorldPosition()
 ]
}
